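# Sound cue for the Knights Castle door closing, a spot environment effect in the Asteria sound set.
# A single wave player feeds a modulator at default settings, which is wired to all four inputs of the root quality-level node, so every quality level sounds the same.
# It uses the shared Fort doors attenuation and sound class at reduced volume, with a concurrency override that prevents new instances once the limit is reached.
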SoundCue /Game/Sounds/Asteria/Environment_Actors/KnightsCastle_Door/Spot_Knights_Castle_Door_Close_Cue
root=QualityLevel_0
nodes [
 {id=Modulator_2, children [WavePlayer_0]}
 {id=QualityLevel_0, children [Modulator_2,Modulator_2,Modulator_2,Modulator_2]}
 {id=WavePlayer_0, properties {soundWaveAssetPtr=&/Spot_Kinghts_Castle_Door_Close}}
]
properties {
 attenuationSettings=/Game/Sounds/Attenuation/Fort_Doors_Att
 bOverrideConcurrency=true
 maxDistance=2400
 soundClassObject=&Fort_Doors
 volumeMultiplier=0.4
 concurrencyOverrides {resolutionRule=PreventNew}
}
